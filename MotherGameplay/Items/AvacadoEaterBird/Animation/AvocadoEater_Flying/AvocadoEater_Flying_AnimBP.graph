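# Animation Blueprint for the flying Avocado Eater Bird item, playing idle, chase, attach, detach and death animations.
# A conduit-hub state machine, entered through a ref-pose DirtyHackState, picks states from the owning bird actor's State enum, mostly via update-set booleans, with Idle as fallback.
# Chase rules read the enum directly, leaving IsInChase unused.
# Its output is layered from alien_root over a pose copied from the attached parent mesh, then inertialized.
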AnimBlueprint /MotherGameplay/Items/AvacadoEaterBird/Animation/AvocadoEater_Flying/AvocadoEater_Flying_AnimBP
target_skeleton=/MotherGameplay/Items/AvacadoEaterBird/Meshes/Avocado_Eater_Bird_Flying/Avocado_Eater_Bird_Flying_Mammal_Skeleton
graphs [
 AnimGraph {nodes [
   n1 = Copy Pose From Mesh() {bCopyCurves=true, bUseAttachedParent=true}
   AnimGraphNode_StateMachine_0 = AvocadoEater_Flying_Statemachine() {is=title AnimGraphNode_StateMachine}
   Output Pose(Result=Inertialization!(Source:Pose Link Structure=Layered blend per bone(BasePose=@n1, BlendPoses_0=@AnimGraphNode_StateMachine_0) {layerSetup [{branchFilters [{boneName=alien_root}]}]}) -> (Pose:Pose Link Structure) {is=title AnimGraphNode_Inertialization}) {layerGroup=None}
  ]}
 AvocadoEater_Flying_Statemachine {
  entry=DirtyHackState
  path=AnimGraph.AnimGraphNode_StateMachine_0.AvocadoEater_Flying_Statemachine
  ids=id
  nodes [
   AnimStateConduitNode_0 = Conduit()
   AnimStateNode_0 = DirtyHackState() {is=title AnimStateNode}
   AnimStateNode_1 = Idle()
   AnimStateNode_2 = Chase() {is=title AnimStateNode}
   AnimStateNode_3 = AttachToTarget() {is=title AnimStateNode}
   AnimStateNode_4 = DetachFromTarget() {is=title AnimStateNode}
   AnimStateNode_5 = Expired() {is=title AnimStateNode}
   AnimStateTransitionNode_0 = AnimStateTransitionNode()
   AnimStateTransitionNode_1 = AnimStateTransitionNode() {crossfadeDuration=0.35, logicType=TLT_Inertialization, priorityOrder=100}
   AnimStateTransitionNode_2 = AnimStateTransitionNode()
   AnimStateTransitionNode_3 = AnimStateTransitionNode() {crossfadeDuration=0.4, logicType=TLT_Inertialization}
   AnimStateTransitionNode_4 = AnimStateTransitionNode()
   AnimStateTransitionNode_5 = AnimStateTransitionNode() {crossfadeDuration=0.3, logicType=TLT_Inertialization, priorityOrder=2}
   AnimStateTransitionNode_6 = AnimStateTransitionNode()
   AnimStateTransitionNode_7 = AnimStateTransitionNode() {crossfadeDuration=0.35, logicType=TLT_Inertialization, priorityOrder=3}
   AnimStateTransitionNode_8 = AnimStateTransitionNode()
   AnimStateTransitionNode_9 = AnimStateTransitionNode() {crossfadeDuration=0.35, logicType=TLT_Inertialization}
  ]
  links [
   {
    dst=Conduit
    rule=AnimGraph.AnimGraphNode_StateMachine_0.AvocadoEater_Flying_Statemachine.AnimStateTransitionNode_6.Transition
    src=AttachToTarget
    via=AnimStateTransitionNode_6
   }
   {
    dst=Conduit
    rule=AnimGraph.AnimGraphNode_StateMachine_0.AvocadoEater_Flying_Statemachine.AnimStateTransitionNode_4.Transition
    src=Chase
    via=AnimStateTransitionNode_4
   }
   {
    dst=AttachToTarget
    rule=AnimGraph.AnimGraphNode_StateMachine_0.AvocadoEater_Flying_Statemachine.AnimStateTransitionNode_5.Transition
    src=Conduit
    via=AnimStateTransitionNode_5
   }
   {
    dst=Chase
    rule=AnimGraph.AnimGraphNode_StateMachine_0.AvocadoEater_Flying_Statemachine.AnimStateTransitionNode_3.Transition
    src=Conduit
    via=AnimStateTransitionNode_3
   }
   {
    dst=DetachFromTarget
    rule=AnimGraph.AnimGraphNode_StateMachine_0.AvocadoEater_Flying_Statemachine.AnimStateTransitionNode_7.Transition
    src=Conduit
    via=AnimStateTransitionNode_7
   }
   {
    dst=Expired
    rule=AnimGraph.AnimGraphNode_StateMachine_0.AvocadoEater_Flying_Statemachine.AnimStateTransitionNode_9.Transition
    src=Conduit
    via=AnimStateTransitionNode_9
   }
   {
    dst=Idle
    rule=AnimGraph.AnimGraphNode_StateMachine_0.AvocadoEater_Flying_Statemachine.AnimStateTransitionNode_1.Transition
    src=Conduit
    via=AnimStateTransitionNode_1
   }
   {
    dst=Conduit
    rule=AnimGraph.AnimGraphNode_StateMachine_0.AvocadoEater_Flying_Statemachine.AnimStateTransitionNode_8.Transition
    src=DetachFromTarget
    via=AnimStateTransitionNode_8
   }
   {
    dst=Conduit
    rule=AnimGraph.AnimGraphNode_StateMachine_0.AvocadoEater_Flying_Statemachine.AnimStateTransitionNode_0.Transition
    src=DirtyHackState
    via=AnimStateTransitionNode_0
   }
   {
    dst=Conduit
    rule=AnimGraph.AnimGraphNode_StateMachine_0.AvocadoEater_Flying_Statemachine.AnimStateTransitionNode_2.Transition
    src=Idle
    via=AnimStateTransitionNode_2
   }
  ]
 }
 Conduit {
  path=AnimGraph.AnimGraphNode_StateMachine_0.AvocadoEater_Flying_Statemachine.AnimStateConduitNode_0.Conduit
  nodes [Result(bCanEnterTransition=true)]
 }
 DirtyHackState {
  path=AnimGraph.AnimGraphNode_StateMachine_0.AvocadoEater_Flying_Statemachine.AnimStateNode_0.DirtyHackState
  nodes [Output Animation Pose(Result=Local Space Ref Pose())]
 }
 Idle {
  path=AnimGraph.AnimGraphNode_StateMachine_0.AvocadoEater_Flying_Statemachine.AnimStateNode_1.Idle
  nodes [Output Animation Pose(Result=Sequence Player 'AvocadoEater_Flying_Idle'() {sequence=&/AvocadoEater_Flying_Idle})]
 }
 Chase {
  path=AnimGraph.AnimGraphNode_StateMachine_0.AvocadoEater_Flying_Statemachine.AnimStateNode_2.Chase
  nodes [Output Animation Pose(Result=Sequence Player 'AvocadoEater_Flying_Charge'() {sequence=&/AvocadoEater_Flying_Charge})]
 }
 AttachToTarget {
  path=AnimGraph.AnimGraphNode_StateMachine_0.AvocadoEater_Flying_Statemachine.AnimStateNode_3.AttachToTarget
  nodes [Output Animation Pose(Result=Sequence Player 'AvocadoEater_Flying_Attach'() {sequence=&/AvocadoEater_Flying_Attach})]
 }
 DetachFromTarget {
  path=AnimGraph.AnimGraphNode_StateMachine_0.AvocadoEater_Flying_Statemachine.AnimStateNode_4.DetachFromTarget
  nodes [Output Animation Pose(Result=Sequence Player 'AvocadoEater_Flying_Detach'() {sequence=&/AvocadoEater_Flying_Detach})]
 }
 Expired {
  path=AnimGraph.AnimGraphNode_StateMachine_0.AvocadoEater_Flying_Statemachine.AnimStateNode_5.Expired
  nodes [Output Animation Pose(Result=Sequence Player 'AvocadoEater_Flying_Death'() {sequence=&/AvocadoEater_Flying_Death})]
 }
 Transition {
  path=AnimGraph.AnimGraphNode_StateMachine_0.AvocadoEater_Flying_Statemachine.AnimStateTransitionNode_0.Transition
  nodes [Result(bCanEnterTransition=true)]
 }
 Transition {
  path=AnimGraph.AnimGraphNode_StateMachine_0.AvocadoEater_Flying_Statemachine.AnimStateTransitionNode_1.Transition
  nodes [Result(bCanEnterTransition=true)]
 }
 Transition {
  path=AnimGraph.AnimGraphNode_StateMachine_0.AvocadoEater_Flying_Statemachine.AnimStateTransitionNode_2.Transition
  nodes [Result(bCanEnterTransition=true)]
 }
 Transition {
  path=AnimGraph.AnimGraphNode_StateMachine_0.AvocadoEater_Flying_Statemachine.AnimStateTransitionNode_3.Transition
  nodes [
   n1 = Get AthenaAvacadoEaterBird() -> (AthenaAvacadoEaterBird:BGA Athena Avacado Eater Bird Parent Object Reference)
   Result(bCanEnterTransition="Equal (Enum)"(
     A:AvacadoEaterBird_States_Enum Enum=Get State!(self:BGA Athena Avacado Eater Bird Parent Object Reference=@n1) -> (State:AvacadoEaterBird_States_Enum Enum=NewEnumerator0)
     B:AvacadoEaterBird_States_Enum Enum=NewEnumerator1
    ))
  ]
 }
 Transition {
  path=AnimGraph.AnimGraphNode_StateMachine_0.AvocadoEater_Flying_Statemachine.AnimStateTransitionNode_4.Transition
  nodes [
   n1 = Get State(self=Get AthenaAvacadoEaterBird()) -> (State=NewEnumerator0)
   Result(bCanEnterTransition=NOT Boolean(A="Equal (Enum)"(A=@n1, B=NewEnumerator1)))
  ]
 }
 Transition {
  path=AnimGraph.AnimGraphNode_StateMachine_0.AvocadoEater_Flying_Statemachine.AnimStateTransitionNode_5.Transition
  nodes [Result(bCanEnterTransition=Get AttachToTarget() -> (AttachToTarget:bool))]
 }
 Transition {
  path=AnimGraph.AnimGraphNode_StateMachine_0.AvocadoEater_Flying_Statemachine.AnimStateTransitionNode_6.Transition
  nodes [Result(bCanEnterTransition=NOT Boolean(A=Get AttachToTarget()))]
 }
 Transition {
  path=AnimGraph.AnimGraphNode_StateMachine_0.AvocadoEater_Flying_Statemachine.AnimStateTransitionNode_7.Transition
  nodes [Result(bCanEnterTransition=Get DetachFromTarget() -> (DetachFromTarget:bool))]
 }
 Transition {
  path=AnimGraph.AnimGraphNode_StateMachine_0.AvocadoEater_Flying_Statemachine.AnimStateTransitionNode_8.Transition
  nodes [Result(bCanEnterTransition=NOT Boolean(A=Get DetachFromTarget()))]
 }
 Transition {
  path=AnimGraph.AnimGraphNode_StateMachine_0.AvocadoEater_Flying_Statemachine.AnimStateTransitionNode_9.Transition
  nodes [Result(bCanEnterTransition=Get IsExpired() -> (IsExpired:bool))]
 }
 EventGraph {nodes [
   Event BlueprintUpdateAnimation()
   > n1 = Sequence!(execute) -> (then_0, then_1, then_2:Exec, then_3:Exec)
   n2 = Get AthenaAvacadoEaterBird()
   > n3 = Branch(Condition="Equal (Enum)"(A=Get State(self=@n2) -> (State=NewEnumerator0), B=NewEnumerator1))
   > Set IsInChase(IsInChase:bool=true) -> (Output_Get:bool)
   Set IsInChase(execute=@n3.else, IsInChase=false)
   n4 = Get AthenaAvacadoEaterBird()
   n5 = Branch(execute=@n1.then_1, Condition="Equal (Enum)"(A=Get State(self=@n4) -> (State=NewEnumerator0), B=NewEnumerator2))
   > Set AttachToTarget(AttachToTarget:bool=true) -> (Output_Get:bool)
   Set AttachToTarget(execute=@n5.else, AttachToTarget=false)
   n6 = Get AthenaAvacadoEaterBird()
   n7 = Branch(execute=@n1.then_2, Condition="Equal (Enum)"(A=Get State(self=@n6) -> (State=NewEnumerator0), B=NewEnumerator3))
   > Set DetachFromTarget(DetachFromTarget:bool=true) -> (Output_Get:bool)
   Set DetachFromTarget(execute=@n7.else, DetachFromTarget=false)
   n8 = Get AthenaAvacadoEaterBird()
   n9 = Branch(execute=@n1.then_3, Condition="Equal (Enum)"(A=Get State(self=@n8) -> (State=NewEnumerator0), B=NewEnumerator6))
   > Set IsExpired(IsExpired:bool=true) -> (Output_Get:bool)
   Set IsExpired(execute=@n9.else, IsExpired=false)
   Event BlueprintInitializeAnimation()
   > n10 = Cast To BGA_Athena_AvacadoEaterBird_Parent!(execute, Object=GetOwningActor()) -> (then, CastFailed, AsBGA Athena Avacado Eater Bird Parent:BGA Athena Avacado Eater Bird Parent Object Reference) {is=title K2Node_DynamicCast}
   > Set AthenaAvacadoEaterBird(AthenaAvacadoEaterBird:BGA Athena Avacado Eater Bird Parent Object Reference=@n10) -> (Output_Get:BGA Athena Avacado Eater Bird Parent Object Reference)
   TryGetPawnOwner()
  ]}
]
variables [
 AthenaAvacadoEaterBird {type=BGA_Athena_AvacadoEaterBird_Parent_C, default=null}
 AttachToTarget {type=boolean}
 DetachFromTarget {type=boolean}
 IsExpired {type=boolean}
 IsInChase {type=boolean}
]
